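# Translucent unlit Niagara master material for sprites, ribbons and meshes, drawing stylized energy with a colored body and outline.
# Two masked noises and a tapering V-gradient scale warp the shape UVs, which drive shape, stroke and scrolling flow textures.
# Static switches select texture, sphere-mask or strokeless shapes and a post-glow composite, while forward and mobile paths always use the stroked composite.
# Dynamic parameters supply time and an optional interpolation WPO.
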Material /Game/Effects/Niagara/MasterMaterials/StencilGlow/M_BTR_Energy_NoCustomDepth
graph {
 properties {
  bEnableResponsiveAA=true
  bUsedWithNiagaraMeshParticles=true
  bUsedWithNiagaraRibbons=true
  bUsedWithNiagaraSprites=true
  blendMode=BLEND_Translucent
  opacityMaskClipValue=0.001
  pixelDepthOffsetMode=PDOM_Legacy
  shadingModel=MSM_Unlit
  usedShadingModels=Unlit
 }
 nodes [
  n1 = DynamicParameter() {desc=Parameters set by Niagara, parameterIndex=1, paramNames [X,Y,Z,WPOIntensity]}
  n2 = Subtract(
   A=MaterialFunctionCall(X=@n1.X, Y=@n1.X, Z=@n1.X) {materialFunction=&MakeFloat3}.Result
   B=ParticlePositionWS()
  )
  n3 = ScalarParameter() {defaultValue=1000, group=Interpolate, parameterName=Interpolate_Clamp}
  Clamp(None=@n2, Min=Multiply(A=@n2, B=@n3) {constA=-1}, Max=@n3) {maxDefault=1000, minDefault=-1000}
  Constant()
  n4 = NamedRerouteUsage()
  n5 = Saturate(None=Multiply(A=@n4) {constB=100})
  n6 = Subtract(A=@n5) {constB=0.98}
  n7 = TextureObjectParameter() {
   group=EnergyFlow
   parameterName=EnergyFlow_Texture
   samplerType=SAMPLERTYPE_LinearGrayscale
   sortPriority=0
   texture=/Game/Effects/Fort_Effects/Emotes/UltraLight/Texture/T_Plasma_B_Greyscale
  }
  n8 = VectorParameter() {group=EnergyFlow, parameterName=EnergyFlow_UVControl, channelNames {a=VPan, b=UPan, g=VScale, r=UScale}, defaultValue=<rgba 2 1 0 0.2>}
  n9 = VectorParameter() {group=uv, parameterName=UVOffset, channelNames {a=VPan, b=UPan, g=VScale, r=UScale}, defaultValue=<rgba 0 -0.05 0 0>}
  n10 = VectorParameter() {group=MainShape, parameterName=MainShape_UVControl, channelNames {a=VPan, b=UPan, g=VScale, r=UScale}, defaultValue=<rgba 0.8 0.6 0 0>}
  n11 = NamedRerouteUsage()
  n12 = StaticBool() {value=true}
  n13 = MaterialFunctionCall(Float2=NamedRerouteUsage()) {materialFunction=&BreakOutFloat2Components}
  n14 = ScalarParameter() {group=MainShapeSizeGradient, parameterName=MainShapeSizeGradient_GrayAdjust_Add}
  n15 = Multiply(
   A=Saturate(None=Add(A=@n13.G, B=@n14))
   B=ScalarParameter() {defaultValue=1, group=MainShapeSizeGradient, parameterName=MainShapeSizeGradient_GrayAdjust_Mul}
  )
  n16 = LinearInterpolate(
   A=ScalarParameter() {parameterName=MainShapeSizeGradient_TipScale}
   B=ScalarParameter() {defaultValue=1.3529, parameterName=MainShapeSizeGradient_TailScale}
   Alpha=Power(
    Base=Saturate(None=@n15)
    Exp=ScalarParameter() {defaultValue=1, group=MainShapeSizeGradient, parameterName=MainShapeSizeGradient_GrayAdjust_Pow}
   )
  ) {constA=55}
  n17 = Add(A=OneMinus(None=@n13.G), B=ScalarParameter() {group=MainShapeSizeGradient, parameterName=MainShapeSizeGradientInvert_GrayAdjust_Add})
  n18 = Saturate(None=Multiply(
    A=Saturate(None=@n17)
    B=ScalarParameter() {defaultValue=1, group=MainShapeSizeGradient, parameterName=MainShapeSizeGradientInvert_GrayAdjust_Mul}
   ))
  n19 = ScalarParameter() {defaultValue=1, group=MainShapeSizeGradient, parameterName=MainShapeSizeGradientInvert_GrayAdjust_Pow}
  n20 = LinearInterpolate(
   A=@n16
   B=LinearInterpolate(A=ScalarParameter() {defaultValue=1, parameterName=MainShapeSizeGradient_InvertScale}, B=@n16, Alpha=Power(Base=@n18, Exp=@n19)) {constA=55}
   Alpha=ScalarParameter() {parameterName=MainShapeSizeGradient_UseInvert}
  )
  n21 = MaterialFunctionCall(UVs=NamedRerouteUsage(), Texture Scale=AppendVector(A=@n20, B=Constant() {r=1})) {materialFunction=&ScaleUVsByCenter}
  n22 = Add(
   A=Add(
    A=AppendVector(A=@n9.UScale, B=@n9.UScale)
    B=MaterialFunctionCall!(
     Texture
     AdditionalUVOffset
     UVControl_RGB=@n10
     UVControl_A=@n10
     "[Override]TimeInput"=@n11
     "[Override]ScaleUVByCenter"=@n12
     "[Override]UVInput"=@n21.UVs
    ) {materialFunction=/Game/Effects/Niagara/MasterMaterials/Functions/Common/FXCommon_TextureSample_V2}.UV
   )
   B=NamedRerouteUsage()
  )
  n23 = VectorParameter() {parameterName=Opcaityr2, defaultValue=<rgba 1.5 3 1 0>}
  n24 = MaterialFunctionCall(
   Texture=MaterialFunctionCall(Texture=@n7, UVControl_RGB=@n8, UVControl_A=@n8, "[Override]TimeInput"=NamedRerouteUsage(), "[Override]UVInput"=@n22) {materialFunction=&FXCommon_TextureSample}.R
   RGB=@n23.RGB
   A=@n23.RGB
  ) {materialFunction=&MF_FX_Refine}
  n25 = Power(Base=NamedRerouteUsage()) {constExponent=3}
  n26 = Multiply(A=Multiply(A=@n24.Result, B=OneMinus(None=@n25)), B=ScalarParameter() {defaultValue=0.3, parameterName=PostGlowBaseStr})
  n27 = VectorParameter() {group=Compose, parameterName=OpacityRefine, defaultValue=<rgba 9 2 0.3 0.5>}
  n28 = MaterialFunctionCall(Texture=@n4, RGB=@n27.RGB, A=@n27.RGB) {materialFunction=&MF_FX_Refine}
  n29 = Multiply(A=Ceil(None=@n4), B=Multiply(A=@n24.Result, B=ScalarParameter() {defaultValue=0.2, group=Compose, parameterName=BaseOpacity})) {constB=0.1}
  n30 = Add(A=NamedRerouteUsage(), B=Saturate(None=Add(A=@n28.Result, B=@n29) {constB=0.1}))
  n31 = StaticSwitchParameter(True=Add(A=Saturate(None=@n6), B=Multiply(A=@n5, B=@n26)), False=@n30) {parameterName=UsePostGlow}
  n32 = ParticleColor()
  DepthFade(
   Opacity=Multiply(A=ShadingPathSwitch(Default=@n31, Deferred=@n31, Forward=@n30, Mobile=@n30), B=@n32.A)
   FadeDistance=ScalarParameter() {defaultValue=20, parameterName=DF}
  )
  n33 = VectorParameter() {group=Compose, parameterName=EnergyColor, defaultValue=<rgba 0 0.6667 1 1>}
  n34 = VectorParameter() {group=Compose, parameterName=ColorRefine, defaultValue=<rgba 9 5 0.3 0.5>}
  n35 = MaterialFunctionCall(Texture=@n4, RGB=@n34.RGB, A=@n34.RGB) {materialFunction=&MF_FX_Refine}
  n36 = LinearInterpolate(
   A=Multiply(A=Add(A=@n35.Result), B=@n33.RGB)
   B=VectorParameter() {group=Compose, parameterName=OutlineColor, defaultValue=<rgba 0 0 0 1>}.RGB
   Alpha=Ceil(None=NamedRerouteUsage())
  )
  n37 = StaticSwitchParameter(True=@n33.RGB, False=@n36) {parameterName=UsePostGlow}
  n38 = ShadingPathSwitch(Default=@n37, Deferred=@n37, Forward=@n36, Mobile=@n36)
  MaterialFunctionCall(Emissive=Divide(A=Multiply(A=@n38, B=@n32.RGB))) {materialFunction=&MF_FX_EmissiveAdjustments}
  n39 = TextureCoordinate()
  n40 = MaterialFunctionCall(Float2=@n39) {materialFunction=&BreakOutFloat2Components}
  n41 = OneMinus(None=@n40.R)
  NamedRerouteDeclaration(None=StaticSwitchParameter(
    True=MaterialFunctionCall(X=@n40.G, Y=@n41) {materialFunction=&MakeFloat2}.Result
    False=@n39
   ) {defaultValue=true, parameterName=UseRibbonUV}) {name=TheUV, nodeColor=<rgba 0.1016 1 0 1>}
  n42 = TextureObjectParameter() {
   group=NoiseA
   parameterName=NoiseA_Texture
   samplerType=SAMPLERTYPE_LinearGrayscale
   sortPriority=0
   texture=/Game/Effects/Fort_Effects/Emotes/UltraLight/Texture/T_DirectionalNoise_Greyscale
  }
  n43 = VectorParameter() {group=NoiseA, parameterName=NoiseA_UVControl, channelNames {a=VPan, b=UPan, g=VScale, r=UScale}, defaultValue=<rgba 0.55 0.7 0 0.7>}
  n44 = NamedRerouteUsage()
  n45 = NamedRerouteUsage()
  n46 = MaterialFunctionCall(Texture=@n42, UVControl_RGB=@n43, UVControl_A=@n43, "[Override]TimeInput"=@n44, "[Override]UVInput"=@n45) {materialFunction=&FXCommon_TextureSample}
  n47 = VectorParameter() {
   group=NoiseA
   parameterName=NoiseA_NoiseControl
   sortPriority=60
   channelNames {a=VStrength, b=UStrength, g=VOffset, r=UOffset}
   defaultValue=<rgba 0 0 0 15>
  }
  n48 = MaterialFunctionCall!(
   NoiseValue=LinearInterpolate(A=AppendVector(A=@n46.R, B=@n46.R), B=@n46.RG, Alpha=ScalarParameter() {parameterName=NoiseA_UseRG})
   Control_RGB=@n47
   Control_A=@n47
  ) {materialFunction=/Game/Effects/Niagara/MasterMaterials/Functions/Common/FXCommon_NoiseControl}
  n49 = VectorParameter() {group=NoiseB, parameterName=NoiseB_UVControl, channelNames {a=VPan, b=UPan, g=VScale, r=UScale}, defaultValue=<rgba 2.64 1.57 0 0.36>}
  n50 = MaterialFunctionCall(
   Texture=TextureObjectParameter() {group=NoiseB, parameterName=NoiseB_Texture, sortPriority=0, texture=/Game/Items/Traps/Textures/T_PerlinRGB_01_M}
   UVControl_RGB=@n49
   UVControl_A=@n49
   "[Override]TimeInput"=@n44
   "[Override]UVInput"=@n45
  ) {materialFunction=&FXCommon_TextureSample}
  n51 = VectorParameter() {
   group=NoiseB
   parameterName=NoiseB_NoiseControl
   sortPriority=60
   channelNames {a=VStrength, b=UStrength, g=VOffset, r=UOffset}
   defaultValue=<rgba -0.5 -0.5 0 0>
  }
  n52 = MaterialFunctionCall(
   NoiseValue=LinearInterpolate(A=AppendVector(A=@n50.R, B=@n50.R), B=@n50.RG, Alpha=ScalarParameter() {parameterName=NoiseB_UseRG})
   Control_RGB=@n51
   Control_A=@n51
  ) {materialFunction=&FXCommon_NoiseControl}
  n53 = MaterialFunctionCall(Float2=NamedRerouteUsage()) {materialFunction=&BreakOutFloat2Components}
  n54 = OneMinus(None=@n53.G)
  n55 = ScalarParameter() {parameterName=NoiseMask_InverV}
  n56 = Power(
   Base=Sine(None=Multiply(A=@n53.R) {constB=0.5})
   Exp=ScalarParameter() {defaultValue=4, group=NoiseMask, parameterName=NoiseMask_CenterAreaPower}
  )
  n57 = ScalarParameter() {defaultValue=4, group=NoiseMask, parameterName=NoiseMask_CenterAreaVGradientPower}
  n58 = Multiply(
   A=Power(
    Base=LinearInterpolate(A=@n54, B=@n53.G, Alpha=@n55)
    Exp=ScalarParameter() {defaultValue=2, group=NoiseMask, parameterName=NoiseMask_Vpower}
   ) {constExponent=1.5}
   B=LinearInterpolate(A=OneMinus(None=@n56), Alpha=Power(Base=@n54, Exp=@n57) {constExponent=1.5})
  )
  n59 = Constant() {r=1}
  NamedRerouteDeclaration(None=Multiply(A=Add(A=@n48.NoiseValue, B=@n52.NoiseValue), B=StaticSwitchParameter(True=@n58, False=@n59) {defaultValue=true, parameterName=UseNoiseMask})) {name=Noise, nodeColor=<rgba 1 0 0.3516 1>}
  n60 = TextureObjectParameter() {
   group=MainShape
   parameterName=MainShape_Texture
   samplerType=SAMPLERTYPE_Normal
   sortPriority=0
   texture=/Game/Effects/Niagara/MasterMaterials/StencilGlow/Texture/T_OutlineShape_Sphere
  }
  n61 = TextureSample(UVs=@n22, Tex=@n60) {samplerType=SAMPLERTYPE_Normal, texture=&T_DirectionalNoise_Greyscale}
  n62 = AppendVector(A=@n61.R, B=@n61.R)
  n63 = VectorParameter() {parameterName=MainShape_TilingMethod, defaultValue=<rgba 1 1 0 1>}
  n64 = AppendVector(A=@n63.R, B=@n63.R)
  n65 = MaterialFunctionCall(Float2=Multiply(
    A=ConstantBiasScale(None=@n62)
    B=MaterialFunctionCall!(TexCoord=@n22, ClampChannel=@n64) {materialFunction=/Game/Effects/Fort_Effects/MaterialFunctions/MF_FX_TilingMethod}.Result
   )) {materialFunction=&BreakOutFloat2Components}
  n66 = SphereMask(
   A=MaterialFunctionCall(UVs=@n22, Texture Scale=ScalarParameter() {defaultValue=0.55, parameterName=MainShape_SphereMaskUVScale}) {materialFunction=&ScaleUVsByCenter}.UVs
   B=Constant() {r=0.5}
  ) {attenuationRadius=0.5, hardnessPercent=0}
  n67 = VectorParameter() {parameterName=MainShape_SphereMask_Refine, defaultValue=<rgba 4 2 0.01 0.5>}
  n68 = Multiply(
   A=Saturate(None=Multiply(A=@n66) {constB=100})
   B=MaterialFunctionCall(Texture=OneMinus(None=@n66), RGB=@n67.RGB, A=@n67.RGB) {materialFunction=&MF_FX_Refine}.Result
  )
  n69 = MaterialFunctionCall(UVs=@n22, Texture Scale=ScalarParameter() {defaultValue=0.5, parameterName=MainShape_NoStrokeUVScale}) {materialFunction=&ScaleUVsByCenter}
  n70 = TextureObjectParameter() {
   group=MainShape
   parameterName=MainShape_NoStrokeTexture
   samplerType=SAMPLERTYPE_LinearGrayscale
   sortPriority=0
   texture=/Game/Effects/Niagara/MasterMaterials/StencilGlow/Texture/NewOutline_NoStroke
  }
  n71 = TextureSample(UVs=@n69.UVs, Tex=@n70) {samplerType=SAMPLERTYPE_LinearGrayscale, texture=&T_DirectionalNoise_Greyscale}
  n72 = MaterialFunctionCall(TexCoord=@n69.UVs, ClampChannel=@n64) {materialFunction=&MF_FX_TilingMethod}
  n73 = StaticSwitchParameter(
   True=@n65.G
   False=StaticSwitchParameter(True=Saturate(None=@n68), False=Multiply(A=@n71.R, B=@n72.Result)) {parameterName=UseSphereMaskGlowShape}
  ) {parameterName=UseTextureStroke}
  NamedRerouteDeclaration(None=ShadingPathSwitch(Default=@n73, Deferred=@n73, Forward=@n65.G, Mobile=@n65.G)) {name=ShapeGradient, nodeColor=<rgba 1 0.8438 0 1>}
  n74 = StaticSwitchParameter(True=@n65.R, False=Constant()) {parameterName=UseTextureStroke}
  NamedRerouteDeclaration(None=ShadingPathSwitch(Default=@n74, Deferred=@n74, Forward=@n65.R, Mobile=@n65.R)) {name=ShapeOutline, nodeColor=<rgba 0.4531 0 1 1>}
  n75 = NamedRerouteUsage()
  n76 = TextureObjectParameter() {
   group=StrokeNoise
   parameterName=StrokeNoise_Texture
   samplerType=SAMPLERTYPE_LinearGrayscale
   sortPriority=0
   texture=&T_SimplePerlin_Greyscale
  }
  n77 = VectorParameter() {
   group=StrokeNoise
   parameterName=StrokeNoise_UVControl
   channelNames {a=VPan, b=UPan, g=VScale, r=UScale}
   defaultValue=<rgba 0.5 0.5 0 0.544>
  }
  n78 = Add(
   A=MaterialFunctionCall(Texture=@n76, UVControl_RGB=@n77, UVControl_A=@n77, "[Override]TimeInput"=NamedRerouteUsage(), "[Override]UVInput"=NamedRerouteUsage()) {materialFunction=&FXCommon_TextureSample}.R
   B=ScalarParameter() {group=StrokeNoise, parameterName=StrokeNoise_GrayAdjust_Add}
  )
  n79 = ScalarParameter() {defaultValue=1.5, group=StrokeNoise, parameterName=StrokeNoise_GrayAdjust_Mul}
  n80 = Power(
   Base=Saturate(None=Multiply(A=@n78, B=@n79))
   Exp=ScalarParameter() {defaultValue=5, group=StrokeNoise, parameterName=StrokeNoise_GrayAdjust_Power}
  )
  n81 = Multiply(
   A=LinearInterpolate(
    A=ScalarParameter() {defaultValue=0.9, group=Stroke, parameterName=Stroke_MaxSize}
    B=ScalarParameter() {defaultValue=0.8, group=Stroke, parameterName=Stroke_MinSize}
    Alpha=Saturate(None=@n80)
   )
   B=Min(A=@n20)
  )
  n82 = Saturate(None=Subtract(A=Saturate(None=@n75), B=@n81))
  n83 = OneMinus(None=@n81)
  n84 = Saturate(None=Multiply(A=Divide(A=@n82, B=@n83), B=ScalarParameter() {defaultValue=10, group=Stroke, parameterName=Stroke_Hardness}))
  n85 = StaticSwitchParameter(True=@n84, False=Constant()) {parameterName=UseTextureStroke}
  NamedRerouteDeclaration(None=ShadingPathSwitch(Default=@n85, Deferred=@n85, Forward=@n84, Mobile=@n84)) {name=FinalOutline, nodeColor=<rgba 1 0.8203 0 1>}
  n86 = Time() {bOverride_Period=true, period=6666}
  n87 = DynamicParameter() {defaultValue=<rgba 1 0 0 1>, paramNames [RandomTime, UseNiagaraTime, NiagaraTime, Param4]}
  NamedRerouteDeclaration(None=Add(A=LinearInterpolate(A=@n86, B=@n87.NiagaraTime, Alpha=@n87.NiagaraTime), B=@n87.RandomTime)) {name=NSTime, nodeColor=<rgba 0.7813 0 1 1>}
  n88 = ComponentMask(None=TextureCoordinate()) {r=true}
  OneMinus(None=@n88)
  n89 = Saturate(None=Power(Base=@n88, Exp=ScalarParameter() {defaultValue=3, group=Interpolate, parameterName=Interpolate_Power}))
  StaticSwitchParameter(True=Multiply(A=@n2, B=Multiply(A=@n1.WPOIntensity, B=@n89)), False=Constant()) {group=Interpolate, parameterName=Use Interpolate WPO}
 ]
}
