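# Translucent unlit master material for Niagara sprites and ribbons that draws an inner-edge glow on objects matching a custom stencil index.
# A custom HLSL node takes jittered radial stencil samples, with radius scaled by particle-camera distance and step count by pixel depth.
# Inverted, stencil-masked results drive refined color and opacity over a faint fill.
# Depth fade, scene-versus-custom-depth difference, sphere or UV-edge masks and particle alpha shape opacity; mobile outputs zero.
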Material /Game/Effects/Niagara/MasterMaterials/StencilGlow/M_BTR_StencilGlow
graph {
 properties {
  bUsedWithNiagaraRibbons=true
  bUsedWithNiagaraSprites=true
  blendMode=BLEND_Translucent
  pixelDepthOffsetMode=PDOM_Legacy
  shadingModel=MSM_Unlit
  usedShadingModels=Unlit
 }
 nodes [
  n1 = ComponentMask(None=SceneTexture() {sceneTextureId=PPI_CustomStencil}.Color) {r=true}
  n2 = ScalarParameter() {defaultValue=2.5, parameterName=EdgeThickness}
  n3 = Multiply(A=SceneTexelSize(), B=ScalarParameter() {defaultValue=1440, parameterName=ScreenSizeThickScaleFactor}) {constB=1080}
  n4 = MaterialFunctionCall(Vector 3=Subtract(A=ParticlePositionWS(), B=CameraPositionWS())) {materialFunction=/Engine/Functions/Engine_MaterialFunctions02/Utility/VectorLength}
  MaterialExpressionAdd_0 = Add(
   A=Multiply(A=Clamp(None=@n1) {maxDefault=0}, B=PixelDepth().R)
   B=Divide(A=Divide(A=@n2, B=@n3), B=Clamp(None=@n4.V3 Length) {clampMode=CMODE_ClampMin, minDefault=30})
  )
  ComponentMask(None=Divide(A=@MaterialExpressionAdd_0)) {r=true}
  SceneTexture() {bFiltered=true, sceneTextureId=PPI_CustomStencil}
  n5 = LinearInterpolate(
   A=VectorParameter() {parameterName=CC, defaultValue=<rgba 0 0.6667 1 1>}.RGB
   B=ParticleColor().RGB
   Alpha=DynamicParameter() {defaultValue=<rgba 0 1 1 1>, paramNames [UseParticleColor, Param2, Param3, Param4]}.UseParticleColor
  )
  MaterialExpressionScalarParameter_10 = ScalarParameter() {defaultValue=27, parameterName=CustomStencilIndex}
  n6 = Saturate(None=If(A=@n1, B=@MaterialExpressionScalarParameter_10, "A > B"=Constant(), "A == B"=Constant() {r=1}, "A < B"=Constant()) {equalsThreshold=0.01})
  n7 = PixelDepth()
  n8 = VectorParameter() {parameterName=SampleDepthDistanceOptimizaiton, defaultValue=<rgba 20 100 0 1>}
  n9 = Saturate(None=Divide(A=Subtract(A=@n7.R, B=@n8.R), B=@n8.G) {constB=2048})
  MaterialExpressionLinearInterpolate_1 = LinearInterpolate(A=ScalarParameter() {defaultValue=3, parameterName=DStep_Near}, B=ScalarParameter() {defaultValue=6, parameterName=DStep_Far}, Alpha=@n9)
  MaterialExpressionLinearInterpolate_2 = LinearInterpolate(A=ScalarParameter() {defaultValue=3, parameterName=RStep_Near}, B=ScalarParameter() {defaultValue=6, parameterName=RStep_Far}, Alpha=@n9)
  MaterialExpressionConstant_4 = Constant() {r=1}
  MaterialExpressionConstant_5 = Constant() {r=1}
  MaterialExpressionTextureObject_0 = TextureObject() {samplerType=SAMPLERTYPE_LinearGrayscale, texture=/Engine/EngineMaterials/Good64x64TilingNoiseHighFreq}
  MaterialExpressionConstant_3 = Constant() {r=1}
  MaterialExpressionConstant_2 = Constant() {r=0.62}
  n10 = Custom!(
   Distance=@MaterialExpressionAdd_0
   DistanceSteps=@MaterialExpressionLinearInterpolate_1
   RadialSteps=@MaterialExpressionLinearInterpolate_2
   TempAARotation=@MaterialExpressionConstant_4
   DistanceMask=@MaterialExpressionConstant_5
   Tex=@MaterialExpressionTextureObject_0
   TempAADistance=@MaterialExpressionConstant_3
   RadialOffset=@MaterialExpressionConstant_2
   StencilIndex=@MaterialExpressionScalarParameter_10
  ) {
   code="half3 CurColor = MaterialFloat3(0.00000000,0.00000000,0.00000000);\r\nhalf2 BaseUV = MaterialFloat2(ScreenAlignedPosition(Parameters.ScreenPosition).xy);\r\nhalf2 NewUV = BaseUV;\r\nhalf StepSize = Distance / (int) DistanceSteps;\r\nhalf CurDistance=0;\r\nhalf2 CurOffset = MaterialFloat2(0.00000000,0.00000000);\r\nhalf TwoPi = 6.283185;\r\nhalf Substep = 0;\r\nhalf2 ScenePixels=View.BufferSizeAndInvSize.xy*BaseUV;\r\nScenePixels+=View.TemporalAAParams.r;\r\nhalf2 RandomSamp = ((uint)(ScenePixels.x) + 2 * (uint)(ScenePixels.y)) % 5;\r\nRandomSamp+=Texture2DSample(Tex,TexSampler,ScenePixels);\r\nRandomSamp/=5;\r\nRandomSamp-=0.5;\r\nTempAARotation*=RandomSamp.x;\r\nTempAADistance*=StepSize*RandomSamp.x;\r\nhalf tempdist = 0;\r\nhalf totaldist = 0;\r\n\r\nint i=0;\r\n##if FEATURE_LEVEL >= FEATURE_LEVEL_SM4\r\nif (DistanceSteps<1)\r\n{\r\nreturn 0;\r\n}\r\nelse\r\n{\r\n\r\nwhile ( i < (int) DistanceSteps)\r\n{\r\n\r\nfor (int j = 0; j < (int) RadialSteps; j++)\r\n{\r\nCurOffset.x = cos(TwoPi*((TempAARotation+Substep) / RadialSteps));\r\nCurOffset.y = sin(TwoPi*((TempAARotation+Substep) / RadialSteps));\r\nCurOffset *=DistanceMask;\r\nNewUV.x = BaseUV.x + (CurOffset.x * (CurDistance+(RandomSamp*TempAADistance)));\r\nNewUV.y = BaseUV.y + (CurOffset.y * (CurDistance+(RandomSamp*TempAADistance)));\r\n\t\ttempdist = (length(NewUV-BaseUV)/Distance);\r\n\t\ttotaldist += tempdist;\r\n        half tmpstencil = (SceneTextureLookup(NewUV, 25, false)).r;\r\n        half stencilmask = saturate(ceil(tmpstencil-StencilIndex+0.1));\r\nCurColor += stencilmask*tempdist;\r\nSubstep++;\r\n}\r\nCurDistance+=StepSize;\r\nSubstep+=RadialOffset;\r\ni++;\r\n}\r\nCurColor = CurColor /totaldist;\r\nreturn CurColor;\r\n}\r\n##else\r\nreturn 0;\r\n##endif\r\n\r\n"
   desc="\n*** Original source before expression upgrade ***\nfloat3 CurColor;\r\nfloat2 BaseUV = MaterialFloat2(ScreenAlignedPosition(Parameters.ScreenPosition).xy);\r\nfloat2 NewUV = BaseUV;\r\nfloat StepSize = Distance / (int) DistanceSteps;\r\nfloat CurDistance=0;\r\nfloat2 CurOffset;\r\nfloat TwoPi = 6.283185;\r\nfloat Substep;\r\nfloat2 ScenePixels=View.RenderTargetSize*BaseUV;\r\nScenePixels+=View.TemporalAAParams.r;\r\nfloat2 RandomSamp = ((uint)(ScenePixels.x) + 2 * (uint)(ScenePixels.y)) % 5;\r\nRandomSamp+=Texture2DSample(Tex,TexSampler,ScenePixels);\r\nRandomSamp/=5;\r\nRandomSamp-=0.5;\r\nTempAARotation*=RandomSamp;\r\nTempAADistance*=StepSize*RandomSamp;\r\nfloat tempdist = -0;\r\nfloat totaldist = 0;\r\n\r\nint i=0;\r\nif (DistanceSteps<1)\r\n{\r\nreturn DecodeSceneColorForMaterialNode(NewUV);\r\n}\r\nelse\r\n{\r\n//CurDistance += 0.5*StepSize;\r\nwhile ( i < (int) DistanceSteps)\r\n{\r\n\r\n//CurDistance+=StepSize;\r\nfor (int j = 0; j < (int) RadialSteps; j++)\r\n{\r\nCurOffset.x = cos(TwoPi*((TempAARotation+Substep) / RadialSteps));\r\nCurOffset.y = sin(TwoPi*((TempAARotation+Substep) / RadialSteps));\r\nCurOffset *=DistanceMask;\r\nNewUV.x = BaseUV.x + (CurOffset.x * (CurDistance+(RandomSamp*TempAADistance)));\r\nNewUV.y = BaseUV.y + (CurOffset.y * (CurDistance+(RandomSamp*TempAADistance)));\r\n\t\ttempdist = (length(NewUV-BaseUV)/Distance);\r\n\t\ttotaldist += tempdist;\r\n\r\nCurColor += DecodeSceneColorForMaterialNode(NewUV)*tempdist;\r\n//CurDistance+=(StepSize+(TempAADistance))/RadialSteps;\r\nSubstep++;\r\n}\r\nCurDistance+=StepSize;\r\nSubstep+=RadialOffset;\r\ni++;\r\n}\r\nCurColor = CurColor /totaldist;\r\nreturn CurColor;\r\n}\n*** Original source before expression upgrade ***\nfloat3 CurColor;\r\nfloat2 BaseUV = MaterialFloat2(ScreenAlignedPosition(Parameters.ScreenPosition).xy);\r\nfloat2 NewUV = BaseUV;\r\nfloat StepSize = Distance / (int) DistanceSteps;\r\nfloat CurDistance=0;\r\nfloat2 CurOffset;\r\nfloat TwoPi = 6.283185;\r\nfloat Substep;\r\nfloat2 ScenePixels=Frame.RenderTargetSize*BaseUV;\r\nScenePixels+=Frame.TemporalAAParams.r;\r\nfloat2 RandomSamp = ((uint)(ScenePixels.x) + 2 * (uint)(ScenePixels.y)) % 5;\r\nRandomSamp+=Texture2DSample(Tex,TexSampler,ScenePixels);\r\nRandomSamp/=5;\r\nRandomSamp-=0.5;\r\nTempAARotation*=RandomSamp;\r\nTempAADistance*=StepSize*RandomSamp;\r\nfloat tempdist = -0;\r\nfloat totaldist = 0;\r\n\r\nint i=0;\r\nif (DistanceSteps<1)\r\n{\r\nreturn DecodeSceneColorForMaterialNode(NewUV);\r\n}\r\nelse\r\n{\r\n//CurDistance += 0.5*StepSize;\r\nwhile ( i < (int) DistanceSteps)\r\n{\r\n\r\n//CurDistance+=StepSize;\r\nfor (int j = 0; j < (int) RadialSteps; j++)\r\n{\r\nCurOffset.x = cos(TwoPi*((TempAARotation+Substep) / RadialSteps));\r\nCurOffset.y = sin(TwoPi*((TempAARotation+Substep) / RadialSteps));\r\nCurOffset *=DistanceMask;\r\nNewUV.x = BaseUV.x + (CurOffset.x * (CurDistance+(RandomSamp*TempAADistance)));\r\nNewUV.y = BaseUV.y + (CurOffset.y * (CurDistance+(RandomSamp*TempAADistance)));\r\n\t\ttempdist = (length(NewUV-BaseUV)/Distance);\r\n\t\ttotaldist += tempdist;\r\n\r\nCurColor += DecodeSceneColorForMaterialNode(NewUV)*tempdist;\r\n//CurDistance+=(StepSize+(TempAADistance))/RadialSteps;\r\nSubstep++;\r\n}\r\nCurDistance+=StepSize;\r\nSubstep+=RadialOffset;\r\ni++;\r\n}\r\nCurColor = CurColor /totaldist;\r\nreturn CurColor;\r\n}"
   inputs [
    {inputName=Distance, input {expression=&self:MaterialExpressionAdd_0}}
    {
     inputName=DistanceSteps
     input {expression=&self:MaterialExpressionLinearInterpolate_1}
    }
    {
     inputName=RadialSteps
     input {expression=&self:MaterialExpressionLinearInterpolate_2}
    }
    {
     inputName=TempAARotation
     input {expression=&self:MaterialExpressionConstant_4}
    }
    {inputName=DistanceMask, input {expression=&self:MaterialExpressionConstant_5}}
    {inputName=Tex, input {expression=&self:MaterialExpressionTextureObject_0}}
    {
     inputName=TempAADistance
     input {expression=&self:MaterialExpressionConstant_3}
    }
    {inputName=RadialOffset, input {expression=&self:MaterialExpressionConstant_2}}
    {
     inputName=StencilIndex
     input {expression=&self:MaterialExpressionScalarParameter_10}
    }
   ]
  }
  n11 = OneMinus(None=@n10)
  n12 = Saturate(None=Multiply(A=ComponentMask(None=@n6) {b=true, g=true, r=true}, B=ComponentMask(None=@n11) {b=true, g=true, r=true}))
  n13 = VectorParameter() {parameterName=ColorRefine, defaultValue=<rgba 1 25 1 0.5>}
  n14 = Add(A=MaterialFunctionCall(Texture=@n12, RGB=@n13.RGB, A=@n13.RGB) {materialFunction=&MF_FX_Refine}.Result)
  n15 = MaterialFunctionCall(Emissive=Divide(A=Multiply(A=@n5, B=@n14))) {materialFunction=&MF_FX_EmissiveAdjustments}
  ShadingPathSwitch(Default=@n15.Result, Deferred=@n15.Result, Mobile=Constant()) {desc=Migrated from FeatureLevelSwitch}
  n16 = ComponentMask(None=Multiply(A=ScalarParameter() {defaultValue=0.1, parameterName=BaseOpacity}, B=@n6)) {b=true, g=true, r=true}
  n17 = VectorParameter() {parameterName=AlphaRefine, defaultValue=<rgba 1 1 1 0.5>}
  n18 = ComponentMask(None=Saturate(None=MaterialFunctionCall(Texture=@n12, RGB=@n17.RGB, A=@n17.RGB) {materialFunction=&MF_FX_Refine}.Result)) {b=true, g=true, r=true}
  n19 = DepthFade(Opacity=Saturate(None=Add(A=@n16, B=@n18)), FadeDistance=ScalarParameter() {defaultValue=10, parameterName=DepthFade})
  n20 = Divide(
   A=Subtract(A=SceneDepth().R, B=SceneTexture() {sceneTextureId=PPI_CustomDepth}.Color)
   B=ScalarParameter() {defaultValue=10, parameterName=CustomDepthFade}
  ) {constB=50}
  n21 = ComponentMask(None=Power(Base=Saturate(None=@n20)) {constExponent=3}) {b=true, g=true, r=true}
  n22 = TextureCoordinate()
  n23 = SphereMask(A=@n22, B=Constant() {r=0.5}, Radius=ScalarParameter() {defaultValue=0.5, parameterName=SphereOpacityMaskRadius}) {attenuationRadius=0.5, hardnessPercent=0}
  n24 = MaterialFunctionCall(Float2=@n22) {materialFunction=&BreakOutFloat2Components}
  n25 = VectorParameter() {group=EdgeFade, parameterName=EdgeFade_Control, channelNames {a="", b=Power, g=Multiply, r=Offset}, defaultValue=<rgba 0 8.5 2 1>}
  n26 = Multiply(
   A=MaterialFunctionCall!(UorV=@n24.R, EdgeControl=@n25) {materialFunction=/Game/Effects/Niagara/MasterMaterials/Functions/Common/FXCommon_EdgeMask_Simple}.Mask
   B=MaterialFunctionCall(UorV=@n24.G, EdgeControl=@n25) {materialFunction=&FXCommon_EdgeMask_Simple}.Mask
  )
  n27 = Multiply(
   A=Multiply(A=Multiply(A=@n19, B=@n21), B=StaticSwitchParameter(True=@n23, False=@n26) {parameterName=UseSphereOpacityMask})
   B=ParticleColor().A
  )
  ShadingPathSwitch(Default=@n27, Deferred=@n27, Mobile=Constant()) {desc=Migrated from FeatureLevelSwitch}
  StaticBoolParameter() {defaultValue=true}
 ]
}
