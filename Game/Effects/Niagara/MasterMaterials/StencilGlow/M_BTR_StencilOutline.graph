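# Translucent unlit Niagara material that draws an outline around objects matching a custom stencil value, near the particle's depth.
# Four screen-space taps detect edges from custom depth differences, gated by custom-versus-scene depth comparisons.
# Noise, pixel depth, resolution and a dynamic parameter drive line width.
# Particle alpha, a sphere or UV-edge mask and depth fade shape the result, which is zero on forward and mobile paths.
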Material /Game/Effects/Niagara/MasterMaterials/StencilGlow/M_BTR_StencilOutline
graph {
 properties {
  bEnableResponsiveAA=true
  bUsedWithNanite=true
  bUsedWithNiagaraMeshParticles=true
  bUsedWithNiagaraRibbons=true
  bUsedWithNiagaraSprites=true
  blendMode=BLEND_Translucent
  opacityMaskClipValue=0.5
  pixelDepthOffsetMode=PDOM_Legacy
  shadingModel=MSM_Unlit
  usedShadingModels=Unlit
 }
 nodes [
  Multiply()
  n1 = ComponentMask(None=ScreenPosition().ViewportUV) {g=true, r=true}
  n2 = ViewSize()
  n3 = Constant() {r=320}
  n4 = Ceil(None=Subtract(A=Divide(A=@n2, B=@n3), B=Constant() {r=2}) {constB=0})
  n5 = ScalarParameter() {defaultValue=1, group=Cartoon, parameterName=CRTNO_Outline-InnerSize1}
  n6 = AppendVector(A=@n5, B=@n5)
  n7 = TextureObjectParameter() {
   group=LineWidthNoise
   parameterName=LineWidthNoise_Texture
   samplerType=SAMPLERTYPE_LinearGrayscale
   sortPriority=0
   texture=&T_SimplePerlin_Greyscale
  }
  n8 = VectorParameter() {
   group=LineWidthNoise
   parameterName=LineWidthNoise_UVControl
   channelNames {a=VPan, b=UPan, g=VScale, r=UScale}
   defaultValue=<rgba 1 0.5 0.3 0.5>
  }
  n9 = MaterialFunctionCall(Float2=TextureCoordinate()) {materialFunction=&BreakOutFloat2Components}
  n10 = OneMinus(None=@n9.R)
  n11 = VectorParameter() {
   group=LineWidthNoise
   parameterName=LineWidthNoise_NoiseControl
   sortPriority=60
   channelNames {a=VStrength, b=UStrength, g=VOffset, r=UOffset}
   defaultValue=<rgba -0.5 -0.5 5 5>
  }
  n12 = MaterialFunctionCall!(
   NoiseValue=MaterialFunctionCall!(
    Texture=@n7
    AdditionalUVOffset
    UVControl_RGB=@n8
    UVControl_A=@n8
    "[Override]TimeInput"
    "[Override]ScaleUVByCenter"
    "[Override]UVInput"=MaterialFunctionCall(X=@n9.G, Y=@n10) {materialFunction=&MakeFloat2}.Result
   ) {materialFunction=/Game/Effects/Niagara/MasterMaterials/Functions/Common/FXCommon_TextureSample_V2}.RG
   Control_RGB=@n11
   Control_A=@n11
  ) {materialFunction=/Game/Effects/Niagara/MasterMaterials/Functions/Common/FXCommon_NoiseControl}
  n13 = ScalarParameter() {group=Cartoon, parameterName=CRTNO_Outline-InnerSize2}
  n14 = Subtract(A=PixelDepth().R, B=ScalarParameter() {defaultValue=100, group=Cartoon, parameterName=CRTNO_Outline-InnerScaleStart, sliderMax=200})
  n15 = ScalarParameter() {defaultValue=1500, group=Cartoon, parameterName=CRTNO_Outline-InnerScaleDistance, sliderMax=1000}
  n16 = LinearInterpolate(
   A=Max(A=Add(A=@n6, B=@n12.NoiseValue), B=ScalarParameter() {defaultValue=0.5, parameterName=LineWidthNoise_MinSize})
   B=AppendVector(A=@n13, B=@n13)
   Alpha=Saturate(None=Divide(A=@n14, B=@n15))
  )
  n17 = Multiply(A=@n16, B=ScalarParameter() {defaultValue=1.5, group=Cartoon, parameterName=CRTNO_Outline-InnerSizeMobileMultiplier})
  n18 = ShadingPathSwitch(Default=@n16, Forward=@n17, Mobile=@n17)
  n19 = Constant() {r=960}
  n20 = ComponentMask(None=ViewSize()) {r=true}
  n21 = Saturate(None=Floor(None=Divide(A=@n19, B=@n20))) {desc=if resolution is smaller than 960H}
  n22 = AppendVector(A=@n21, B=@n21)
  n23 = ComponentMask(None=Multiply(
    A=Clamp(None=@n4) {desc="if resolution is bigger than 960 =2", maxDefault=2, minDefault=1}
    B=Max(A=@n18, B=@n22) {desc="If res smaller than target, calmp min value to 1"}
   ) {constA=2}) {g=true, r=true}
  n24 = DynamicParameter() {parameterIndex=3, paramNames [Param1,Param2,Param3,OutlineThick]}
  n25 = ViewProperty() {property=MEVP_ResolutionFraction}
  n26 = Multiply(A=SceneTexelSize(), B=Multiply(A=Multiply(A=@n23, B=@n24.OutlineThick), B=AppendVector(A=@n25.Property, B=@n25.Property)))
  n27 = Constant()
  n28 = AppendVector(A=ComponentMask(None=@n26) {desc=x, r=true}, B=@n27)
  n29 = Add(A=@n1, B=@n28) {desc=x+}
  n30 = ComponentMask(None=SceneTexture(UVs=@n29) {sceneTextureId=PPI_CustomDepth}.Color) {r=true}
  n31 = SceneDepth(UVs=@n29)
  n32 = Constant()
  n33 = Constant() {r=1}
  n34 = AppendVector(A=@n27, B=ComponentMask(None=@n26) {desc=y, g=true}) {desc=y+}
  n35 = Add(A=@n1, B=@n34) {desc=y+}
  n36 = ComponentMask(None=SceneTexture(UVs=@n35) {sceneTextureId=PPI_CustomDepth}.Color) {r=true}
  n37 = SceneDepth(UVs=@n35)
  n38 = Constant()
  n39 = Constant() {r=1}
  n40 = Constant() {r=-1}
  n41 = Add(A=@n1, B=Multiply(A=@n28, B=@n40)) {desc=x-}
  n42 = ComponentMask(None=SceneTexture(UVs=@n41) {sceneTextureId=PPI_CustomDepth}.Color) {r=true}
  n43 = Add(
   A=Add(
    A=If(A=@n30, B=@n31.R, "A > B"=@n32, "A < B"=@n33) {equalsThreshold=0.01}
    B=If(A=@n36, B=@n37.R, "A > B"=@n38, "A < B"=@n39) {equalsThreshold=0.01}
   )
   B=If(A=@n42, B=SceneDepth(UVs=@n41).R, "A > B"=Constant(), "A < B"=Constant() {r=1}) {equalsThreshold=0.01}
  )
  n44 = Add(A=@n1, B=Multiply(A=@n40, B=@n34)) {desc=y-}
  n45 = ComponentMask(None=SceneTexture(UVs=@n44) {sceneTextureId=PPI_CustomDepth}.Color) {r=true}
  n46 = If(A=@n45, B=SceneDepth(UVs=@n44).R, "A > B"=Constant(), "A < B"=Constant() {r=1}) {equalsThreshold=0.01}
  n47 = ComponentMask(None=SceneTexture() {bFiltered=true, sceneTextureId=PPI_CustomDepth}.Color) {r=true}
  n48 = SceneDepth()
  n49 = Constant()
  n50 = Constant() {r=1}
  n51 = Saturate(None=Add(A=Add(A=@n43, B=@n46), B=If(A=@n47, B=@n48.R, "A > B"=@n49, "A < B"=@n50) {equalsThreshold=0.01}))
  n52 = SceneTexture() {bFiltered=true, sceneTextureId=PPI_CustomStencil}
  n53 = ScalarParameter() {defaultValue=27, parameterName=CustomStencil}
  n54 = Constant() {r=9999999827968}
  n55 = ComponentMask(None=If(A=ComponentMask(None=@n52.Color) {r=true}, B=@n53, "A > B"=@n54, "A == B"=@n47, "A < B"=@n54) {equalsThreshold=0.01}) {r=true}
  n56 = PixelDepth()
  n57 = ScalarParameter() {defaultValue=1000, parameterName=MaxDepth}
  n58 = If(A=Abs(None=Subtract(A=@n55, B=@n56.R)), B=@n57, "A > B"=@n54, "A == B"=@n54, "A < B"=@n55)
  n59 = SceneTexture(UVs=@n29) {sceneTextureId=PPI_CustomStencil}
  n60 = ComponentMask(None=If(A=ComponentMask(None=@n59.Color) {r=true}, B=@n53, "A > B"=@n54, "A == B"=@n30, "A < B"=@n54) {equalsThreshold=0.01}) {r=true}
  n61 = If(A=Abs(None=Subtract(A=@n60, B=@n56.R)), B=@n57, "A > B"=@n54, "A == B"=@n54, "A < B"=@n60)
  n62 = SceneTexture(UVs=@n35) {sceneTextureId=PPI_CustomStencil}
  n63 = ComponentMask(None=If(A=ComponentMask(None=@n62.Color) {r=true}, B=@n53, "A > B"=@n54, "A == B"=@n36, "A < B"=@n54) {equalsThreshold=0.01}) {r=true}
  n64 = If(A=Abs(None=Subtract(A=@n63, B=@n56.R)), B=@n57, "A > B"=@n54, "A == B"=@n54, "A < B"=@n63)
  n65 = SceneTexture(UVs=@n41) {sceneTextureId=PPI_CustomStencil}
  n66 = ComponentMask(None=If(A=ComponentMask(None=@n65.Color) {r=true}, B=@n53, "A > B"=@n54, "A == B"=@n42, "A < B"=@n54) {equalsThreshold=0.01}) {r=true}
  n67 = If(A=Abs(None=Subtract(A=@n66, B=@n56.R)), B=@n57, "A > B"=@n54, "A == B"=@n54, "A < B"=@n66)
  n68 = SceneTexture(UVs=@n44) {sceneTextureId=PPI_CustomStencil}
  n69 = ComponentMask(None=If(A=ComponentMask(None=@n68.Color) {r=true}, B=@n53, "A > B"=@n54, "A == B"=@n45, "A < B"=@n54) {equalsThreshold=0.01}) {r=true}
  n70 = If(A=Abs(None=Subtract(A=@n69, B=@n56.R)), B=@n57, "A > B"=@n54, "A == B"=@n54, "A < B"=@n69)
  n71 = Add(A=Add(A=Subtract(A=@n58, B=@n61), B=Subtract(A=@n58, B=@n64)), B=Add(A=Subtract(A=@n58, B=@n67), B=Subtract(A=@n58, B=@n70)))
  n72 = ScalarParameter() {defaultValue=-500, group=Cartoon, parameterName=Divisor_Min, sliderMax=1, sliderMin=-1}
  n73 = Divide(A=@n71, B=@n72)
  n74 = Saturate(None=Divide(A=@n71, B=Multiply(A=@n72) {constB=-1}))
  n75 = LinearInterpolate(A=Add(A=Saturate(None=@n73), B=@n74), B=@n74, Alpha=ScalarParameter() {parameterName=OnesideOutline})
  n76 = Constant() {r=0.2}
  n77 = Multiply(A=Multiply(A=Subtract(A=@n75, B=@n76), B=Constant() {r=2}), B=ParticleColor().A)
  n78 = TextureCoordinate()
  n79 = SphereMask(A=@n78, B=Constant() {r=0.5}, Radius=ScalarParameter() {defaultValue=0.5, parameterName=SphereOpacityMaskRadius}) {attenuationRadius=0.5, hardnessPercent=0}
  n80 = MaterialFunctionCall(Float2=@n78) {materialFunction=&BreakOutFloat2Components}
  n81 = VectorParameter() {group=EdgeFade, parameterName=EdgeFade_Control, channelNames {a="", b=Power, g=Multiply, r=Offset}, defaultValue=<rgba 0 8.5 2 1>}
  n82 = Multiply(
   A=MaterialFunctionCall!(UorV=@n80.R, EdgeControl=@n81) {materialFunction=/Game/Effects/Niagara/MasterMaterials/Functions/Common/FXCommon_EdgeMask_Simple}.Mask
   B=MaterialFunctionCall(UorV=@n80.G, EdgeControl=@n81) {materialFunction=&FXCommon_EdgeMask_Simple}.Mask
  )
  n83 = DepthFade(
   Opacity=Multiply(A=Multiply(A=@n51, B=@n77), B=StaticSwitchParameter(True=@n79, False=@n82) {parameterName=UseSphereOpacityMask})
   FadeDistance=ScalarParameter() {defaultValue=25, parameterName=DepthFadeDistance}
  )
  n84 = Constant()
  ShadingPathSwitch(Default=@n83, Deferred=@n83, Forward=@n84, Mobile=@n84)
  VectorParameter() {parameterName=OutlineColor, defaultValue=<rgba 0.075 0.1917 0.25 1>}
 ]
}
